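; Food truck prop Blueprint that reacts to damage by triggering a steam effect on its steam mug sign.
; Each hit calls a cosmetic event with Reset false, gated by CanExecuteCosmeticEvents.
; That event writes game time and Reset into the sign mesh's custom primitive data and activates its Niagara steam.
; ResetFlag tracks an active effect window, choosing between two retriggerable delay durations before clearing.
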(event EventBeginPlay
  (EventDispatchers|BindEventtoOnDamaged (delegate Custom|OnDamaged_Event)))

(event Custom|OnDamaged_Event
  (params (DamagedActor :type Actor%20Object%20Reference) (Damage :type Float%20%28single-precision%29) (InstigatedBy :type Controller%20Object%20Reference) (DamageCauser :type Actor%20Object%20Reference) (HitLocation :type Vector) (FHitComponent :type Primitive%20Component%20Object%20Reference) (BoneName :type Name) (Momentum :type Vector))
  (CallFunction|OnDamagedCosmeticEvent false)
  (if (Variables|Default|GetResetFlag)
    (Utilities|FlowControl|RetriggerableDelay (Variables|Default|GetEffect%20Out%20Duration))
    (else
      (Variables|Default|SetResetFlag true)
      (Utilities|FlowControl|RetriggerableDelay (Variables|Default|GetEffect%20in%20Out%20Duration))))
  (Variables|Default|SetResetFlag false))

(event Custom|OnDamagedCosmetic_Event
  (params (Reset :type Boolean))
  (Networking|CanExecuteCosmeticEvents
    (:True
      (Rendering|Material|SetCustomPrimitiveDataVector2 (Variables|Default|GetSM_SteamMug_Sign_A) 2 (Math|Vector2D|MakeVector2D (Utilities|Time|GetGameTimeinSeconds) (Math|Conversions|ToFloat(Boolean) Reset)))
      (Components|Activation|Activate (Variables|Default|GetNS_SteamMug_Sign_A_Steam) Reset))))

(fn ConstructionScript
  (|CallParentFunction))
